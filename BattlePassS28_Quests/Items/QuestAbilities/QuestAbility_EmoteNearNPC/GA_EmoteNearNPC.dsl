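; Quest gameplay ability that detects when the player emotes near specific NPC pawns.
; On activation it waits for the emote gameplay tag on the avatar, then runs a single-hit sphere trace at the avatar's location using the Range variable, ignoring the avatar.
; If the hit Fort pawn carries a SilkWalker or holiday CardboardCrew spawner tag, it broadcasts the EmoteNearPawn custom verb for the avatar.
; A debug print remains, and the end-ability event is empty.
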(event Ability|EventActivateAbility
  (bind _returnvalue (Ability|GetAvatarActorfromActorInfo))
  (Ability|Tasks|WaitGameplayTagAddToActor _returnvalue "(TagName=\"Gameplay.Action.Player.Emote\")"
    (:Added
      (bind _outhit (Collision|SphereTraceForObjects (Transformation|GetActorLocation _returnvalue) (Transformation|GetActorLocation _returnvalue) (Variables|Default|GetRange) (Utilities|Array|MakeArray "ObjectTypeQuery3") false (Utilities|Array|MakeArray _returnvalue)))
      (bind _asfort_pawn (Utilities|Casting|CastToFortPawn (.HitActor (Collision|BreakHitResult _outhit))))
      (Development|PrintString (Utilities|String|Append (Utilities|GetDisplayName self) ": " "FortPawn Found!") false false)
      (if (GameplayTags|HasAnyMatchingGameplayTags _asfort_pawn (GameplayTags|MakeGameplayTagContainerfromArray (Utilities|Array|MakeArray "(TagName=\"AISpawnerData.Type.SilkWalker\")" "(TagName=\"AISpawnerData.Type.SilkWalkerNoVehicle\")" "(TagName=\"AISpawnerData.Type.Tandem.CardboardCrew_Holiday\")")))
        (CustomVerb|BroadcastCustomVerbMessage (CustomVerb|MakeCustomVerbMessage (Ability|GetAvatarActorfromActorInfo) "(TagName=\"CustomVerb.EmoteNearPawn\")"))))))

(event Ability|EventOnEndAbility
  (params (bWasCancelled :type Boolean)))
